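# Physics asset for the Flare Jonesy male body, setting up secondary motion for its scarf and drawstrings.
# Pelvis, spine_05 and thigh capsules form a coarse body proxy around the dangling parts.
# Small master bodies join scarf back and tie to spine_05, drawstrings to pelvis, with free rotation.
# Chains below them use limited swing and tight twist; the boxed scarf back swings less on one axis.
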PhysicsAsset /BRCosmetics/Characters/Player/Male/Medium/Bodies/M_MED_Flare_Jonesy/Meshes/M_MED_Flare_Jonesy_Physics
bodies {
 dyn_draw_strin_l_01 [{shapeType=Capsule, length=1, radius=1, center=<1 0 0>, rotation=<rot 90 0 0>}]
 dyn_draw_strin_l_02 [{shapeType=Capsule, length=0.7709, radius=1.2291, center=<1 0 0>, rotation=<rot 90 0 0>}]
 dyn_draw_strin_l_03 [{shapeType=Capsule, length=1.005, radius=1.3896, center=<1 0 0>, rotation=<rot 90 0 0>}]
 dyn_draw_strin_l_MSTR [{shapeType=Capsule, length=0.505, radius=0.505}]
 dyn_draw_strin_r_01 [{shapeType=Capsule, length=1, radius=1, center=<1 0 0>, rotation=<rot 90 0 0>}]
 dyn_draw_strin_r_02 [{shapeType=Capsule, length=0.6421, radius=1.3579, center=<1 0 -0.1634>, rotation=<rot 90 0 0>}]
 dyn_draw_strin_r_03 [{shapeType=Capsule, length=0.6645, radius=1.3355, center=<1 0 0.1682>, rotation=<rot 90 0 0>}]
 dyn_draw_strin_r_MSTR [{shapeType=Capsule, length=0.505, radius=0.505}]
 dyn_scarf_back_01 [{shapeType=Box, extentX=3, extentY=1, extentZ=8.1395, center=<0.1 0 0>}]
 dyn_scarf_back_02 [{shapeType=Box, extentX=3, extentY=1, extentZ=8.1395, center=<0.1 0 0>}]
 dyn_scarf_back_03 [{shapeType=Box, extentX=3.5, extentY=1, extentZ=8.1395, center=<0.8001 -0.0904 0>}]
 dyn_scarf_back_MSTR1 [{shapeType=Capsule, length=0.505, radius=0.505}]
 dyn_scarf_tie_MSTR [{shapeType=Capsule, length=0.505, radius=0.505}]
 dyn_scarf_tie_l_01 [{shapeType=Capsule, length=0.6686, radius=1.2105, center=<1 0 0>, rotation=<rot 90 0 0>}]
 dyn_scarf_tie_l_02 [{shapeType=Capsule, length=0.8791, radius=1, center=<1 0 0>, rotation=<rot 90 0 0>}]
 dyn_scarf_tie_l_03 [{shapeType=Capsule, length=0.8791, radius=1, center=<1 0 0>, rotation=<rot 90 0 0>}]
 dyn_scarf_tie_r_01 [{shapeType=Capsule, length=2, radius=1, center=<1 0 0>, rotation=<rot 90 0 0>}]
 dyn_scarf_tie_r_02 [{shapeType=Capsule, length=2, radius=1, center=<1 0 0>, rotation=<rot 90 0 0>}]
 pelvis [{shapeType=Capsule, length=8.5493, radius=8.7418, center=<-3.7507 1.5529 0>}]
 spine_05 [{shapeType=Capsule, length=8.4284, radius=10.3654, center=<0.8716 -0.4668 0>}]
 thigh_l [{shapeType=Capsule, length=15.7062, radius=10.1374, center=<-10 0 0>, rotation=<rot 90 0 -1.0822>}]
 thigh_r [{shapeType=Capsule, length=16.1019, radius=9.7797, center=<10 0 0>, rotation=<rot 88.9186 -179.9999 -179.9999>}]
}
constraints [
 {bone1Name=spine_05, bone2Name=pelvis, swing1LimitDegrees=45, swing2LimitDegrees=45, twistLimitDegrees=45}
 {
  bone1Name=dyn_scarf_tie_MSTR
  bone2Name=spine_05
  swing1LimitDegrees=45
  swing1Motion=Free
  swing2LimitDegrees=45
  swing2Motion=Free
  twistLimitDegrees=45
  twistMotion=Free
 }
 {bone1Name=dyn_scarf_tie_l_01, bone2Name=dyn_scarf_tie_MSTR, swing1LimitDegrees=20, swing2LimitDegrees=20, twistLimitDegrees=2}
 {bone1Name=dyn_scarf_tie_l_02, bone2Name=dyn_scarf_tie_l_01, swing1LimitDegrees=30, swing2LimitDegrees=30, twistLimitDegrees=2}
 {bone1Name=dyn_scarf_tie_l_03, bone2Name=dyn_scarf_tie_l_02, swing1LimitDegrees=30, swing2LimitDegrees=30, twistLimitDegrees=2}
 {bone1Name=dyn_scarf_tie_r_01, bone2Name=dyn_scarf_tie_MSTR, swing1LimitDegrees=30, swing2LimitDegrees=30, twistLimitDegrees=2}
 {bone1Name=dyn_scarf_tie_r_02, bone2Name=dyn_scarf_tie_r_01, swing1LimitDegrees=30, swing2LimitDegrees=30, twistLimitDegrees=2}
 {
  bone1Name=dyn_scarf_back_MSTR1
  bone2Name=spine_05
  swing1LimitDegrees=45
  swing1Motion=Free
  swing2LimitDegrees=45
  swing2Motion=Free
  twistLimitDegrees=45
  twistMotion=Free
 }
 {bone1Name=dyn_scarf_back_01, bone2Name=dyn_scarf_back_MSTR1, swing1LimitDegrees=25, swing2LimitDegrees=10, twistLimitDegrees=2}
 {bone1Name=dyn_scarf_back_02, bone2Name=dyn_scarf_back_01, swing1LimitDegrees=25, swing2LimitDegrees=10, twistLimitDegrees=2}
 {bone1Name=dyn_scarf_back_03, bone2Name=dyn_scarf_back_02, swing1LimitDegrees=25, swing2LimitDegrees=10, twistLimitDegrees=2}
 {
  bone1Name=dyn_draw_strin_l_MSTR
  bone2Name=pelvis
  swing1LimitDegrees=45
  swing1Motion=Free
  swing2LimitDegrees=45
  swing2Motion=Free
  twistLimitDegrees=45
  twistMotion=Free
 }
 {bone1Name=dyn_draw_strin_l_01, bone2Name=dyn_draw_strin_l_MSTR, swing1LimitDegrees=30, swing2LimitDegrees=30, twistLimitDegrees=5}
 {bone1Name=dyn_draw_strin_l_02, bone2Name=dyn_draw_strin_l_01, swing1LimitDegrees=30, swing2LimitDegrees=30, twistLimitDegrees=5}
 {bone1Name=dyn_draw_strin_l_03, bone2Name=dyn_draw_strin_l_02, swing1LimitDegrees=30, swing2LimitDegrees=30, twistLimitDegrees=5}
 {
  bone1Name=dyn_draw_strin_r_MSTR
  bone2Name=pelvis
  swing1LimitDegrees=45
  swing1Motion=Free
  swing2LimitDegrees=45
  swing2Motion=Free
  twistLimitDegrees=45
  twistMotion=Free
 }
 {bone1Name=dyn_draw_strin_r_01, bone2Name=dyn_draw_strin_r_MSTR, swing1LimitDegrees=30, swing2LimitDegrees=30, twistLimitDegrees=5}
 {bone1Name=dyn_draw_strin_r_02, bone2Name=dyn_draw_strin_r_01, swing1LimitDegrees=30, swing2LimitDegrees=30, twistLimitDegrees=5}
 {bone1Name=dyn_draw_strin_r_03, bone2Name=dyn_draw_strin_r_02, swing1LimitDegrees=30, swing2LimitDegrees=30, twistLimitDegrees=5}
 {
  bone1Name=thigh_r
  bone2Name=pelvis
  swing1LimitDegrees=45
  swing1Motion=Free
  swing2LimitDegrees=45
  swing2Motion=Free
  twistLimitDegrees=45
  twistMotion=Free
 }
 {
  bone1Name=thigh_l
  bone2Name=pelvis
  swing1LimitDegrees=45
  swing1Motion=Free
  swing2LimitDegrees=45
  swing2Motion=Free
  twistLimitDegrees=45
  twistMotion=Free
 }
]
