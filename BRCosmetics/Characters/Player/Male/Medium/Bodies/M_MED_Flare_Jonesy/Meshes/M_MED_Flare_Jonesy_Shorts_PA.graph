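# Physics asset for the shorts on the Flare Jonesy character, covering only the pelvis and left leg.
# Small capsules on the front, inside and outside shorts bones carry narrow twist and moderate swing limit angles.
# Front and inside attach to the upper thigh twist bone, while outside goes through an intermediate master bone.
# The thigh-to-pelvis and master-to-thigh joints swing and twist freely, and a larger capsule represents the pelvis.
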PhysicsAsset /BRCosmetics/Characters/Player/Male/Medium/Bodies/M_MED_Flare_Jonesy/Meshes/M_MED_Flare_Jonesy_Shorts_PA
bodies {
 pelvis [{shapeType=Capsule, length=9.9419, radius=7.3493, center=<-0.0002 2.1735 0>}]
 shorts_l_front [{shapeType=Capsule, length=2, radius=1, center=<0.5 0 0>, rotation=<rot 90 0 0>}]
 shorts_l_inside [{shapeType=Capsule, length=2, radius=1, center=<0.5 0 0>, rotation=<rot 90 0 0>}]
 shorts_l_outside [{shapeType=Capsule, length=2, radius=1, center=<0.5 0 0>, rotation=<rot 90 0 0>}]
 shorts_l_outside_MSTR [{shapeType=Capsule, length=0.5, radius=0.5}]
 thigh_twist_01_l [{shapeType=Capsule, length=0.5, radius=0.5}]
}
constraints [
 {
  bone1Name=thigh_twist_01_l
  bone2Name=pelvis
  swing1LimitDegrees=45
  swing1Motion=Free
  swing2LimitDegrees=45
  swing2Motion=Free
  twistLimitDegrees=45
  twistMotion=Free
 }
 {bone1Name=shorts_l_front, bone2Name=thigh_twist_01_l, swing1LimitDegrees=35, swing2LimitDegrees=35, twistLimitDegrees=5}
 {bone1Name=shorts_l_inside, bone2Name=thigh_twist_01_l, swing1LimitDegrees=35, swing2LimitDegrees=35, twistLimitDegrees=5}
 {bone1Name=shorts_l_outside, bone2Name=shorts_l_outside_MSTR, swing1LimitDegrees=35, swing2LimitDegrees=35, twistLimitDegrees=5}
 {
  bone1Name=shorts_l_outside_MSTR
  bone2Name=thigh_twist_01_l
  swing1LimitDegrees=45
  swing1Motion=Free
  swing2LimitDegrees=45
  swing2Motion=Free
  twistLimitDegrees=45
  twistMotion=Free
 }
]
