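# Physics asset for the right leg of the shorts on the male medium Flare Jonesy player body, likely for secondary cloth-like motion.
# Small capsules on the front, inside and outside shorts bones each constrain to the right thigh twist bone, with moderate swing and tight twist limits.
# The thigh twist capsule links to a larger pelvis capsule through a constraint with free swing and twist.
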PhysicsAsset /BRCosmetics/Characters/Player/Male/Medium/Bodies/M_MED_Flare_Jonesy/Meshes/M_MED_Flare_Jonesy_Shorts_Right_PA
bodies {
 pelvis [{shapeType=Capsule, length=9.9419, radius=7.3493, center=<-0.0002 2.1735 0>}]
 shorts_r_front [{shapeType=Capsule, length=2, radius=1, center=<0.5 0 0>, rotation=<rot 90 0 0>}]
 shorts_r_inside [{shapeType=Capsule, length=2, radius=1, center=<0.5 0 0>, rotation=<rot 90 0 0>}]
 shorts_r_outside [{shapeType=Capsule, length=2, radius=1, center=<0.5 0 0>, rotation=<rot 90 0 0>}]
 thigh_twist_01_r [{shapeType=Capsule, length=0.5, radius=0.5}]
}
constraints [
 {
  bone1Name=thigh_twist_01_r
  bone2Name=pelvis
  swing1LimitDegrees=45
  swing1Motion=Free
  swing2LimitDegrees=45
  swing2Motion=Free
  twistLimitDegrees=45
  twistMotion=Free
 }
 {bone1Name=shorts_r_front, bone2Name=thigh_twist_01_r, swing1LimitDegrees=35, swing2LimitDegrees=35, twistLimitDegrees=5}
 {bone1Name=shorts_r_inside, bone2Name=thigh_twist_01_r, swing1LimitDegrees=35, swing2LimitDegrees=35, twistLimitDegrees=5}
 {bone1Name=shorts_r_outside, bone2Name=thigh_twist_01_r, swing1LimitDegrees=35, swing2LimitDegrees=35, twistLimitDegrees=5}
]
