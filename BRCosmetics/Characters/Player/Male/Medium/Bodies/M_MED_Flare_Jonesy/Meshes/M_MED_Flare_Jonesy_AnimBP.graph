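# Post-process style cosmetic AnimBP for the Flare Jonesy body, adding secondary motion to the pose copied from its attached parent.
# Chained RigidBody and AnimDynamics nodes simulate the body and shorts, easing body physics on longer frames and shorts physics while parachuting.
# Thigh-to-calf twist deltas drive corrective shorts bone offsets.
# Back scarf bone offsets apply during skydiving, with a brief extra pulse after leaving dive mode.
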AnimBlueprint /BRCosmetics/Characters/Player/Male/Medium/Bodies/M_MED_Flare_Jonesy/Meshes/M_MED_Flare_Jonesy_AnimBP
target_skeleton=&/M_MED_Flare_Jonesy_Skeleton
graphs [
 AnimGraph {nodes [
   n1 = Get calf_l_alpha() -> (calf_l_alpha:double)
   n2 = RigidBody(
    ComponentPose=Local To Component(LocalPose=Copy Pose From Mesh() {bUseAttachedParent=true})
    Alpha=Get rigid_body_alpha() -> (rigid_body_alpha:double)
   ) {
    bOverrideWorldGravity=true
    overridePhysicsAsset=&/M_MED_Flare_Jonesy_Physics
    simulationSpace=BaseBoneSpace
    alphaBoolBlend {blendOption=Linear}
    baseBoneRef {boneName=root}
    componentLinearAccScale=<0.4 0.3 0.3>
    componentLinearVelScale=<0.3 0.3 0.3>
    overrideWorldGravity {z=-980}
   }
   n3 = Get shortsAlpha() -> (shortsAlpha:double)
   n4 = RigidBody(ComponentPose=@n2, Alpha=@n3) {
    bClampLinearTranslationLimitToRefPose=true
    bOverrideWorldGravity=true
    overridePhysicsAsset=&/M_MED_Flare_Jonesy_Shorts_PA
    simulationSpace=BaseBoneSpace
    alphaBoolBlend {blendOption=Linear}
    baseBoneRef {boneName=thigh_l}
    componentLinearAccScale=<0.3 0.2 0.2>
    componentLinearVelScale=<0.2 0.2 0.2>
    overrideWorldGravity {z=-980}
   }
   n5 = RigidBody(ComponentPose=@n4, Alpha=@n3) {
    bClampLinearTranslationLimitToRefPose=true
    bOverrideWorldGravity=true
    overridePhysicsAsset=&/M_MED_Flare_Jonesy_Shorts_Right_PA
    simulationSpace=BaseBoneSpace
    alphaBoolBlend {blendOption=Linear}
    baseBoneRef {boneName=thigh_r}
    componentLinearAccScale=<0.3 0.2 0.2>
    componentLinearVelScale=<0.2 0.2 0.2>
    overrideWorldGravity {z=-980}
   }
   n6 = Anim Dynamics - Bone: shorts_fix_l_back(ComponentPose=@n5, Alpha=1, ExternalForce="0,0,0") {
    angularDampingOverride=0.995
    bGravityOverrideInSimSpace=true
    bOverrideAngularDamping=true
    bUseGravityOverride=true
    alphaBoolBlend {blendOption=Linear}
    boundBone {boneName=shorts_fix_l_back}
    componentAppliedLinearAccClamp=<10000 10000 10000>
    componentLinearAccScale=<0.3 0.2 0.2>
    componentLinearVelScale=<0.2 0.2 0.2>
    gravityOverride {z=-980}
    physicsBodyDefinitions [{
      sphereCollisionRadius=0
      boundBone {boneName=shorts_fix_l_back}
      boxExtents=<5 5 5>
      constraintSetup {angularLimitsMax=<0 5 30>, angularLimitsMin=<0 -5 -10>, angularTarget=<0 0 0>}
      localJointOffset=<0.5 0 0>
     }]
    relativeSpaceBone {boneName=thigh_l}
   }
   n7 = Anim Dynamics - Bone: shorts_fix_r_back(ComponentPose=@n6, Alpha=1, ExternalForce="0,0,0") {
    angularDampingOverride=0.995
    bGravityOverrideInSimSpace=true
    bOverrideAngularDamping=true
    bUseGravityOverride=true
    alphaBoolBlend {blendOption=Linear}
    boundBone {boneName=shorts_fix_r_back}
    componentAppliedLinearAccClamp=<10000 10000 10000>
    componentLinearAccScale=<0.3 0.2 0.2>
    componentLinearVelScale=<0.2 0.2 0.2>
    gravityOverride {z=-980}
    physicsBodyDefinitions [{
      sphereCollisionRadius=0
      boundBone {boneName=shorts_fix_r_back}
      boxExtents=<5 5 5>
      constraintSetup {angularLimitsMax=<0 5 30>, angularLimitsMin=<0 -5 -10>, angularTarget=<0 0 0>}
      localJointOffset=<0.5 0 0>
     }]
    relativeSpaceBone {boneName=thigh_l}
   }
   n8 = "Transform (Modify) Bone - Bone: shorts_fix_l_back"(ComponentPose=@n7, Alpha=@n1, Translation="0.000,-1.820,4.095", Rotation="0.000,0.000,-16.4", Scale="1,1,1") {
    rotationMode=BMM_Additive
    translationMode=BMM_Additive
    alphaBoolBlend {blendOption=Linear}
    boneToModify {boneName=shorts_fix_l_back}
    rotation {roll=-16.4}
    translation {y=-1.82, z=4.095}
   }
   n9 = "Transform (Modify) Bone - Bone: shorts_l_outside_MSTR"(ComponentPose=@n8, Alpha=@n1, Translation="2,0.000,0.000", Rotation="-3,0,-52", Scale="0.8,0.8,0.8") {
    rotationMode=BMM_Additive
    scaleMode=BMM_Replace
    translationMode=BMM_Additive
    alphaBoolBlend {blendOption=Linear}
    boneToModify {boneName=shorts_l_outside_MSTR}
    rotation {pitch=-3, roll=-52}
    scale=<0.8 0.8 0.8>
    translation {x=2}
   }
   n10 = "Transform (Modify) Bone - Bone: shorts_fix_r_back"(
    ComponentPose=@n9
    Alpha=Get calf_r_alpha() -> (calf_r_alpha:double)
    Translation="0.000,5,0.000"
    Rotation="0.000,0.000,-9.258"
    Scale="1,1,1"
   ) {
    rotationMode=BMM_Additive
    translationMode=BMM_Additive
    alphaBoolBlend {blendOption=Linear}
    boneToModify {boneName=shorts_fix_r_back}
    rotation {roll=-9.2583}
    translation {y=5}
   }
   n11 = "Transform (Modify) Bone - Bone: dyn_scarf_back_MSTR1"!(ComponentPose=@n10, bAlphaBoolEnabled:bool=Get bIsSkyDiving(), Translation="2,-1.5,0", Rotation="0,20,0", Scale="1,1,1") -> (Pose) {
    alphaInputType=Bool
    rotationMode=BMM_Additive
    rotationSpace=BCS_BoneSpace
    translationMode=BMM_Additive
    translationSpace=BCS_BoneSpace
    alphaBoolBlend {blendOption=Linear}
    boneToModify {boneName=dyn_scarf_back_MSTR1}
    rotation {yaw=20}
    translation=<2 -1.5>
   }
   n12 = "Transform (Modify) Bone - Bone: dyn_scarf_back_MSTR1"(
    ComponentPose=@n11
    bAlphaBoolEnabled=Get Skydiving_bandana_alpha() -> (Skydiving_bandana_alpha:bool)
    Translation="4,-3,0"
    Rotation="0,50,0"
    Scale="1,1,1"
   ) {
    alphaInputType=Bool
    rotationMode=BMM_Additive
    rotationSpace=BCS_BoneSpace
    translationMode=BMM_Additive
    translationSpace=BCS_BoneSpace
    alphaBoolBlend {blendOption=Linear, blendOutTime=0.1}
    boneToModify {boneName=dyn_scarf_back_MSTR1}
    rotation {yaw=50}
    translation=<4 -3>
   }
   Output Pose(Result=Component To Local(ComponentPose=@n12)) {layerGroup=None}
  ]}
 EventGraph {nodes [
   n1 = Event BlueprintUpdateAnimation()
   n2 = Get skelMesh() -> (skelMesh:Skeletal Mesh Component Object Reference)
   > Set calf_r_alpha(calf_r_alpha:double=GetAlphaFromReferenceTransformDelta(SkeletalMeshComponent=@n2, BoneName=thigh_r, BaseName=calf_r, DeltaComponentType=QuaternionTwist, InputRange=(X=17,Y=115), TwistAxis="0,0,1")) -> (Output_Get:double)
   > Set calf_l_alpha(calf_l_alpha:double=GetAlphaFromReferenceTransformDelta(SkeletalMeshComponent=@n2, BoneName=thigh_l, BaseName=calf_l, DeltaComponentType=QuaternionTwist, InputRange=(X=17,Y=110), TwistAxis="0,0,1")) -> (Output_Get:double)
   > n3 = Sequence()
   > n4 = Branch(Condition=Get bIsSkydiveDiveMode())
   Do Once(execute=@n4.else, Reset=@n4.then, Start Closed=true)
   > Set Skydiving_bandana_alpha(Skydiving_bandana_alpha:bool=true) -> (Output_Get:bool)
   > Delay(Duration=0.3)
   > Set Skydiving_bandana_alpha(Skydiving_bandana_alpha=false)
   Set rigid_body_alpha(
    execute=@n3.then_1
    rigid_body_alpha:double=MapRangeClamped(Value=float * float(A=@n1.DeltaTimeX) {is=title K2Node_PromotableOperator}, InRangeA=0.02, InRangeB=0.028, OutRangeA=1, OutRangeB=0.6)
   ) -> (Output_Get:double)
   > n5 = Branch(Condition=Get bIsParachuteOpened())
   > Set shortsAlpha(shortsAlpha:double=0.5) -> (Output_Get:double)
   Set shortsAlpha(execute=@n5.else, shortsAlpha=1)
   Event BlueprintInitializeAnimation()
   > Set skelMesh(skelMesh:Skeletal Mesh Component Object Reference=GetOwnerSkeletalMeshComponent()) -> (Output_Get:Skeletal Mesh Component Object Reference)
   Get bIsSkyDiving()
  ]}
]
variables [
 calf_l_alpha {type=number}
 calf_r_alpha {type=number}
 rigid_body_alpha {type=number}
 shortsAlpha {type=number}
 skelMesh {type=SkeletalMeshComponent}
 Skydiving_bandana_alpha {type=boolean}
]
