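# Physics asset for the Flare Jonesy medium male body, defining capsule bodies that give the character's hair and beard physical secondary motion.
# A head capsule anchors several front, back, side, top and bottom hair chains, each linked bone to bone in sequence.
# Three beard capsules hang from a jaw body that is itself constrained to the head.
# Swing and twist limits keep twist tight throughout, with the beard held stiffer than the hair.
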PhysicsAsset /BRCosmetics/Characters/Player/Male/Medium/Bodies/M_MED_Flare_Jonesy/Meshes/Parts/M_MED_Flare_Jonesy_Physics
bodies {
 C_jaw [{shapeType=Capsule, length=0.5, radius=0.5}]
 dyn_beard_l [{shapeType=Capsule, length=2, radius=1, center=<1 0 0>, rotation=<rot 90 0 0>}]
 dyn_beard_m [{shapeType=Capsule, length=2, radius=1, center=<1 0 0>, rotation=<rot 90 0 0>}]
 dyn_beard_r [{shapeType=Capsule, length=2, radius=1, center=<1 0 0>, rotation=<rot 90 0 0>}]
 dyn_hair_bck_l_01 [{shapeType=Capsule, length=2, radius=1, center=<0.5 0 0>, rotation=<rot 90 0 0>}]
 dyn_hair_bck_l_02 [{shapeType=Capsule, length=2, radius=1, center=<0.5 0 0>, rotation=<rot 90 0 0>}]
 dyn_hair_bck_l_03 [{shapeType=Capsule, length=2, radius=1, center=<0.5 0 0>, rotation=<rot 90 0 0>}]
 dyn_hair_bck_m_01 [{shapeType=Capsule, length=2, radius=1, center=<0.5 0 0>, rotation=<rot 90 0 0>}]
 dyn_hair_bck_m_02 [{shapeType=Capsule, length=2, radius=1, center=<0.5 0 -0.1005>, rotation=<rot 90 0 0>}]
 dyn_hair_bck_m_03 [{shapeType=Capsule, length=2, radius=1, center=<0.5 0 -0.0525>, rotation=<rot 90 0 0>}]
 dyn_hair_bck_r_01 [{shapeType=Capsule, length=0.9872, radius=1, center=<0.5 0 0>, rotation=<rot 90 0 0>}]
 dyn_hair_bck_r_02 [{shapeType=Capsule, length=0.9872, radius=1, center=<0.5 0 0>, rotation=<rot 90 0 0>}]
 dyn_hair_bck_r_03 [{shapeType=Capsule, length=0.9872, radius=1, center=<0.5 0 0>, rotation=<rot 90 0 0>}]
 dyn_hair_bck_r_04 [{shapeType=Capsule, length=0.9872, radius=1, center=<0.5 0 0>, rotation=<rot 90 0 0>}]
 dyn_hair_btm_r_01 [{shapeType=Capsule, length=1.0482, radius=0.7757, center=<0.5 0 0>, rotation=<rot 90 0 0>}]
 dyn_hair_btm_r_02 [{shapeType=Capsule, length=1.0482, radius=0.7757, center=<0.5 0 0>, rotation=<rot 90 0 0>}]
 dyn_hair_frnt_l_01 [{shapeType=Capsule, length=2, radius=1, center=<0.5 0 0>, rotation=<rot 90 0 0>}]
 dyn_hair_frnt_l_02 [{shapeType=Capsule, length=2, radius=1, center=<0.5 0 0>, rotation=<rot 90 0 0>}]
 dyn_hair_frnt_l_03 [{shapeType=Capsule, length=2, radius=1, center=<0.5 0 0>, rotation=<rot 90 0 0>}]
 dyn_hair_frnt_l_04 [{shapeType=Capsule, length=2, radius=1, center=<0.5 0 0>, rotation=<rot 90 0 0>}]
 dyn_hair_frnt_r_01 [{shapeType=Capsule, length=1.0099, radius=0.8854, center=<0.5 0 0>, rotation=<rot 90 0 0>}]
 dyn_hair_frnt_r_02 [{shapeType=Capsule, length=1.0099, radius=0.8854, center=<0.5 0 0>, rotation=<rot 90 0 0>}]
 dyn_hair_frnt_r_03 [{shapeType=Capsule, length=1.0099, radius=0.8854, center=<0.5 0 0>, rotation=<rot 90 0 0>}]
 dyn_hair_frnt_r_04 [{shapeType=Capsule, length=1.0099, radius=0.8854, center=<0.5 0 0>, rotation=<rot 90 0 0>}]
 dyn_hair_l_01 [{shapeType=Capsule, length=2, radius=1, center=<0.5 0 0>, rotation=<rot 90 0 0>}]
 dyn_hair_l_02 [{shapeType=Capsule, length=2, radius=1, center=<0.5 0 0>, rotation=<rot 90 0 0>}]
 dyn_hair_l_03 [{shapeType=Capsule, length=2, radius=1, center=<0.5 0 0>, rotation=<rot 90 0 0>}]
 dyn_hair_l_04 [{shapeType=Capsule, length=2, radius=1, center=<0.5 0 0>, rotation=<rot 90 0 0>}]
 dyn_hair_top_r_01 [{shapeType=Capsule, length=1.3024, radius=0.634, center=<0.5 0 0>, rotation=<rot 90 0 0>}]
 dyn_hair_top_r_02 [{shapeType=Capsule, length=1.3024, radius=0.634, center=<0.5 0 0>, rotation=<rot 90 0 0>}]
 head [{shapeType=Capsule, length=5.078, radius=6.7351, center=<2.6761 2.5195 0>, rotation=<rot 90 0 0>}]
}
constraints [
 {bone1Name=C_jaw, bone2Name=head, swing1LimitDegrees=45, swing2LimitDegrees=45, twistLimitDegrees=45}
 {bone1Name=dyn_beard_l, bone2Name=C_jaw, swing1LimitDegrees=10, swing2LimitDegrees=10, twistLimitDegrees=2}
 {bone1Name=dyn_beard_r, bone2Name=C_jaw, swing1LimitDegrees=10, swing2LimitDegrees=10, twistLimitDegrees=2}
 {bone1Name=dyn_beard_m, bone2Name=C_jaw, swing1LimitDegrees=10, swing2LimitDegrees=10, twistLimitDegrees=2}
 {bone1Name=dyn_hair_bck_l_01, bone2Name=head, swing1LimitDegrees=35, swing2LimitDegrees=35, twistLimitDegrees=5}
 {bone1Name=dyn_hair_bck_l_02, bone2Name=dyn_hair_bck_l_01, swing1LimitDegrees=35, swing2LimitDegrees=35, twistLimitDegrees=5}
 {bone1Name=dyn_hair_bck_l_03, bone2Name=dyn_hair_bck_l_02, swing1LimitDegrees=35, swing2LimitDegrees=35, twistLimitDegrees=5}
 {bone1Name=dyn_hair_bck_m_01, bone2Name=head, swing1LimitDegrees=35, swing2LimitDegrees=35, twistLimitDegrees=5}
 {bone1Name=dyn_hair_bck_m_02, bone2Name=dyn_hair_bck_m_01, swing1LimitDegrees=35, swing2LimitDegrees=35, twistLimitDegrees=5}
 {bone1Name=dyn_hair_bck_m_03, bone2Name=dyn_hair_bck_m_02, swing1LimitDegrees=35, swing2LimitDegrees=35, twistLimitDegrees=5}
 {bone1Name=dyn_hair_bck_r_01, bone2Name=head, swing1LimitDegrees=35, swing2LimitDegrees=35, twistLimitDegrees=5}
 {bone1Name=dyn_hair_bck_r_02, bone2Name=dyn_hair_bck_r_01, swing1LimitDegrees=25, swing2LimitDegrees=25, twistLimitDegrees=5}
 {bone1Name=dyn_hair_bck_r_03, bone2Name=dyn_hair_bck_r_02, swing1LimitDegrees=25, swing2LimitDegrees=25, twistLimitDegrees=5}
 {bone1Name=dyn_hair_bck_r_04, bone2Name=dyn_hair_bck_r_03, swing1LimitDegrees=35, swing2LimitDegrees=35, twistLimitDegrees=5}
 {bone1Name=dyn_hair_top_r_01, bone2Name=head, swing1LimitDegrees=25, swing2LimitDegrees=25, twistLimitDegrees=5}
 {bone1Name=dyn_hair_top_r_02, bone2Name=dyn_hair_top_r_01, swing1LimitDegrees=25, swing2LimitDegrees=25, twistLimitDegrees=5}
 {bone1Name=dyn_hair_frnt_r_01, bone2Name=head, swing1LimitDegrees=20, swing2LimitDegrees=20, twistLimitDegrees=5}
 {bone1Name=dyn_hair_frnt_r_02, bone2Name=dyn_hair_frnt_r_01, swing1LimitDegrees=25, swing2LimitDegrees=25, twistLimitDegrees=5}
 {bone1Name=dyn_hair_frnt_r_03, bone2Name=dyn_hair_frnt_r_02, swing1LimitDegrees=25, swing2LimitDegrees=25, twistLimitDegrees=5}
 {bone1Name=dyn_hair_frnt_r_04, bone2Name=dyn_hair_frnt_r_03, swing1LimitDegrees=25, swing2LimitDegrees=25, twistLimitDegrees=5}
 {bone1Name=dyn_hair_btm_r_01, bone2Name=head, swing1LimitDegrees=35, swing2LimitDegrees=35, twistLimitDegrees=5}
 {bone1Name=dyn_hair_btm_r_02, bone2Name=dyn_hair_btm_r_01, swing1LimitDegrees=35, swing2LimitDegrees=35, twistLimitDegrees=5}
 {bone1Name=dyn_hair_l_01, bone2Name=head, swing1LimitDegrees=25, swing2LimitDegrees=25, twistLimitDegrees=5}
 {bone1Name=dyn_hair_l_02, bone2Name=dyn_hair_l_01, swing1LimitDegrees=35, swing2LimitDegrees=35, twistLimitDegrees=5}
 {bone1Name=dyn_hair_l_03, bone2Name=dyn_hair_l_02, swing1LimitDegrees=35, swing2LimitDegrees=35, twistLimitDegrees=5}
 {bone1Name=dyn_hair_l_04, bone2Name=dyn_hair_l_03, swing1LimitDegrees=35, swing2LimitDegrees=35, twistLimitDegrees=5}
 {bone1Name=dyn_hair_frnt_l_01, bone2Name=head, swing1LimitDegrees=20, swing2LimitDegrees=20, twistLimitDegrees=5}
 {bone1Name=dyn_hair_frnt_l_02, bone2Name=dyn_hair_frnt_l_01, swing1LimitDegrees=25, swing2LimitDegrees=25, twistLimitDegrees=5}
 {bone1Name=dyn_hair_frnt_l_03, bone2Name=dyn_hair_frnt_l_02, swing1LimitDegrees=35, swing2LimitDegrees=35, twistLimitDegrees=5}
 {bone1Name=dyn_hair_frnt_l_04, bone2Name=dyn_hair_frnt_l_03, swing1LimitDegrees=35, swing2LimitDegrees=35, twistLimitDegrees=5}
]
